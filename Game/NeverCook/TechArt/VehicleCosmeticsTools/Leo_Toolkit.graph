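# Editor Utility Widget in the NeverCook vehicle cosmetics tools folder, a mostly skeletal toolkit panel for use in the editor.
# A scrollable stack holds two dark button rows with dividers.
# The Edit button loads this widget's own asset and opens it for editing through the Asset Editor Subsystem.
# The second button, named for opening a TT vehicle map, keeps a placeholder label and has no click logic here.
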EditorUtilityWidgetBlueprint /Game/NeverCook/TechArt/VehicleCosmeticsTools/Leo_Toolkit
graphs [EventGraph {nodes [
   "On Clicked (btn_EditTool)"() {is=title K2Node_ComponentBoundEvent}
   > n1 = LoadAsset(AssetPath=&self)
   > OpenEditorForAssets!(
    execute
    self:Asset Editor Subsystem Object Reference=Get AssetEditorSubsystem()
    Assets:Array of Object References=Make Array("[0]":Object Reference=@n1) -> (Array:Array of Object References)
    OpenedMethod:EAssetTypeActivationOpenedMethod Enum=Edit
   ) -> (then, ReturnValue:bool)
   Event PreConstruct()
   Event Construct()
   Event Tick()
  ]}]
variables []
widgets {
 info {rootWidgetClass=&StackBox, widgetCount=22}
 widgets [
  StackBox_0 {widgetClassPath=&StackBox, widget=StackBox_0}
  EditorUtilityScrollBox_0 {
   widgetClassPath=&EditorUtilityScrollBox
   parent=StackBox_0
   widget=EditorUtilityScrollBox_0
   slot {class=StackBoxSlot, properties {size {sizeRule=Fill, value=1}}}
  }
  VerticalBox1 {widgetClassPath=&VerticalBox, parent=EditorUtilityScrollBox_0, widget=VerticalBox1, slot {class=ScrollBoxSlot}}
  Overlay_1 {widgetClassPath=&Overlay, parent=VerticalBox1, widget=Overlay_1, slot {class=VerticalBoxSlot}}
  Border_5 {
   widgetClassPath=&Border
   parent=Overlay_1
   widget=Border_5
   properties {
    background {
     bIsDynamicallyLoaded=false
     drawAs=Box
     imageType=NoImage
     mirroring=NoMirror
     resourceName=None
     resourceObject=null
     tiling=NoTile
     imageSize=<32 32>
     margin=<ltrb 0 0 0 0>
     outlineSettings {
      bUseBrushTransparency=false
      roundingType=HalfHeightRadius
      width=0
      color {colorUseRule=UseColor_Specified, specifiedColor=<rgba 0 0 0 0>}
      cornerRadii=<0 0 0 0>
     }
     tintColor {colorUseRule=UseColor_Specified, specifiedColor=<rgba 1 1 1 1>}
     uVRegion {bIsValid=false, max=<0 0>, min=<0 0>}
    }
    brushColor=<rgba 0.026 0.026 0.026 1>
   }
   slot {class=OverlaySlot, properties {horizontalAlignment=HAlign_Fill, verticalAlignment=VAlign_Fill}}
  }
  VerticalBox_LevelPreviewActorControls_2 {widgetClassPath=&VerticalBox, parent=Overlay_1, widget=VerticalBox_LevelPreviewActorControls_2, slot {class=OverlaySlot}}
  HorizontalBox_8 {
   widgetClassPath=&HorizontalBox
   parent=VerticalBox_LevelPreviewActorControls_2
   widget=HorizontalBox_8
   slot {class=VerticalBoxSlot, properties {padding=<ltrb 6 6 6 6>}}
  }
  btn_EditTool {widgetClassPath=&EditorUtilityButton, parent=HorizontalBox_8, widget=btn_EditTool, slot {class=HorizontalBoxSlot}}
  TextBlock_107 {
   widgetClassPath=&TextBlock
   parent=btn_EditTool
   widget=TextBlock_107
   properties {
    text=Edit
    colorAndOpacity {colorUseRule=UseColor_Specified, specifiedColor=<rgba 0.5729 0.5729 0.5729 1>}
    font {
     bForceMonospaced=false
     bMaterialIsStencil=false
     fontMaterial=null
     fontObject=&Roboto
     letterSpacing=0
     monospacedWidth=1
     size=12
     skewAmount=0
     typefaceFontName=Bold
     outlineSettings {
      bApplyOutlineToDropShadows=false
      bMiteredCorners=false
      bSeparateFillAlpha=false
      outlineMaterial=null
      outlineSize=0
      outlineColor=<rgba 0 0 0 1>
     }
    }
   }
   slot {class=ButtonSlot}
  }
  VerticalBox_1 {widgetClassPath=&VerticalBox, parent=VerticalBox1, widget=VerticalBox_1, slot {class=VerticalBoxSlot}}
  Spacer_7 {widgetClassPath=&Spacer, parent=VerticalBox_1, widget=Spacer_7, properties {size=<1 16>}, slot {class=VerticalBoxSlot}}
  Border_6 {
   widgetClassPath=&Border
   parent=VerticalBox_1
   widget=Border_6
   properties {brushColor=<rgba 1 1 1 0.25>}
   slot {class=VerticalBoxSlot, properties {padding=<ltrb 0 16 0 3>}}
  }
  VerticalBox2 {widgetClassPath=&VerticalBox, parent=EditorUtilityScrollBox_0, widget=VerticalBox2, slot {class=ScrollBoxSlot}}
  Overlay {widgetClassPath=&Overlay, parent=VerticalBox2, widget=Overlay, slot {class=VerticalBoxSlot}}
  Border {
   widgetClassPath=&Border
   parent=Overlay
   widget=Border
   properties {
    background {
     bIsDynamicallyLoaded=false
     drawAs=Box
     imageType=NoImage
     mirroring=NoMirror
     resourceName=None
     resourceObject=null
     tiling=NoTile
     imageSize=<32 32>
     margin=<ltrb 0 0 0 0>
     outlineSettings {
      bUseBrushTransparency=false
      roundingType=HalfHeightRadius
      width=0
      color {colorUseRule=UseColor_Specified, specifiedColor=<rgba 0 0 0 0>}
      cornerRadii=<0 0 0 0>
     }
     tintColor {colorUseRule=UseColor_Specified, specifiedColor=<rgba 1 1 1 1>}
     uVRegion {bIsValid=false, max=<0 0>, min=<0 0>}
    }
    brushColor=<rgba 0.026 0.026 0.026 1>
   }
   slot {class=OverlaySlot, properties {horizontalAlignment=HAlign_Fill, verticalAlignment=VAlign_Fill}}
  }
  VerticalBox_LevelPreviewActorControls {widgetClassPath=&VerticalBox, parent=Overlay, widget=VerticalBox_LevelPreviewActorControls, slot {class=OverlaySlot}}
  HorizontalBox {
   widgetClassPath=&HorizontalBox
   parent=VerticalBox_LevelPreviewActorControls
   widget=HorizontalBox
   slot {class=VerticalBoxSlot, properties {padding=<ltrb 6 6 6 6>}}
  }
  btn_OpenTTVehicleMap_1 {widgetClassPath=&EditorUtilityButton, parent=HorizontalBox, widget=btn_OpenTTVehicleMap_1, slot {class=HorizontalBoxSlot}}
  TextBlock {
   widgetClassPath=&TextBlock
   parent=btn_OpenTTVehicleMap_1
   widget=TextBlock
   properties {
    text=Button
    colorAndOpacity {colorUseRule=UseColor_Specified, specifiedColor=<rgba 0.5729 0.5729 0.5729 1>}
    font {
     bForceMonospaced=false
     bMaterialIsStencil=false
     fontMaterial=null
     fontObject=&Roboto
     letterSpacing=0
     monospacedWidth=1
     size=12
     skewAmount=0
     typefaceFontName=Bold
     outlineSettings {
      bApplyOutlineToDropShadows=false
      bMiteredCorners=false
      bSeparateFillAlpha=false
      outlineMaterial=null
      outlineSize=0
      outlineColor=<rgba 0 0 0 1>
     }
    }
   }
   slot {class=ButtonSlot}
  }
  VerticalBox {widgetClassPath=&VerticalBox, parent=VerticalBox2, widget=VerticalBox, slot {class=VerticalBoxSlot}}
  Spacer {widgetClassPath=&Spacer, parent=VerticalBox, widget=Spacer, properties {size=<1 16>}, slot {class=VerticalBoxSlot}}
  Border_1 {
   widgetClassPath=&Border
   parent=VerticalBox
   widget=Border_1
   properties {brushColor=<rgba 1 1 1 0.25>}
   slot {class=VerticalBoxSlot, properties {padding=<ltrb 0 16 0 3>}}
  }
 ]
}
